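# Deferred decal material with Modulate blending for a puddle or wet drip stain, with a darkening value and reduced roughness.
# A tiled, offset stain texture's red channel is multiplied by cosine and linear UV falloffs to fade the decal edges, then sharpened by CheapContrast.
# That mask drives a lerp from Drip Value.
# Multiplied by the powered green channel, it also drives a lerp from Min Roughness.
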Material /Game/Environments/AutumnDecay/Terrain/Materials/M_SZ_Decal_Puddle01
graph {
 properties {blendMode=BLEND_Modulate, materialDomain=MD_DeferredDecal, pixelDepthOffsetMode=PDOM_Legacy, usedShadingModels=Default Lit}
 nodes [
  n1 = TextureCoordinate()
  n2 = ComponentMask(None=@n1) {r=true}
  n3 = Cosine(None=Divide(A=@n2) {constB=1})
  n4 = AppendVector(
   A=ScalarParameter() {defaultValue=0.5, parameterName=X Tile, sliderMax=1}
   B=ScalarParameter() {defaultValue=1, parameterName=Y Tile, sliderMax=1}
  )
  n5 = ScalarParameter() {parameterName=X Translate, sliderMax=1}
  n6 = ScalarParameter() {defaultValue=0.8, parameterName=Y Translate, sliderMax=1}
  n7 = TextureSampleParameter2D(UVs=Add(A=Multiply(A=@n1, B=@n4), B=AppendVector(A=@n5, B=@n6))) {parameterName=StainMask, samplerType=SAMPLERTYPE_Masks, texture=/Game/Environments/Asteria/Sets/NeonCity/Techart/Textures/T_StainA_Masks}
  n8 = Divide(A=ComponentMask(None=@n1) {g=true}) {constB=1.2}
  n9 = Multiply(A=Multiply(A=OneMinus(None=@n3), B=Multiply(A=@n7.R) {constB=3}), B=OneMinus(None=Saturate(None=@n8)))
  n10 = Saturate(None=Divide(A=ComponentMask(None=@n1) {g=true}) {constB=0.05})
  n11 = Saturate(None=MaterialFunctionCall(In=Multiply(A=@n9, B=@n10), Contrast=ScalarParameter() {defaultValue=3, parameterName=Contrast}) {materialFunction=&CheapContrast}.Result)
  n12 = Power(Base=@n7.G)
  LinearInterpolate(A=ScalarParameter() {defaultValue=0.2, parameterName=Min Roughness}, Alpha=OneMinus(None=Multiply(A=@n11, B=@n12)))
  LinearInterpolate(A=ScalarParameter() {defaultValue=0.2, parameterName=Drip Value}, Alpha=OneMinus(None=@n11))
  n13 = Divide(A=Cosine(None=@n2)) {constB=0.4}
  Multiply(A=@n3, B=Saturate(None=OneMinus(None=@n13))) {constB=5}
 ]
}
